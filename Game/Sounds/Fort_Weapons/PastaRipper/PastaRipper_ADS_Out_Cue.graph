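# Sound cue for the PastaRipper weapon's ADS-out (leaving aim-down-sights) foley, currently empty with no root or sound nodes, so it produces no audio.
# It is assigned to the weapon foley sound class at reduced volume and references the shared weapon reload attenuation asset.
# It overrides concurrency to cap simultaneous instances, and its stored attenuation override values have distance attenuation switched off.
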SoundCue /Game/Sounds/Fort_Weapons/PastaRipper/PastaRipper_ADS_Out_Cue
root=null
nodes []
properties {
 attenuationSettings=&Fort_Weapon_Reload_Att
 bOverrideConcurrency=true
 maxDistance=2250
 soundClassObject=&Fort_WeaponFoley
 volumeMultiplier=0.301
 attenuationOverrides {bAttenuate=false, stereoSpread=5000}
 concurrencyOverrides {maxCount=6}
}
