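# Sound cue for the PastaRipper weapon's aim-down-sights mechanical muzzle sound, classed as weapon foley.
# A branch on the IsLocal parameter plays a single ADS-in mech muzzle wave when true; the false and unset inputs are empty, so other cases are silent.
# It plays at reduced volume, uses the shared weapon reload attenuation and overrides concurrency to cap simultaneous instances.
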SoundCue /Game/Sounds/Fort_Weapons/PastaRipper/SC_PastaRipper_ADS_Mech_Muzzle
root=Branch_0
nodes [
 {id=Branch_0, children [WavePlayer_0,null,null], properties {boolParameterName=IsLocal}}
 {id=WavePlayer_0, properties {soundWaveAssetPtr=&/PastaRipper_Fire_ADS_In_Mech_Muzzle_01}}
]
properties {
 attenuationSettings=&Fort_Weapon_Reload_Att
 bOverrideConcurrency=true
 maxDistance=2250
 soundClassObject=&Fort_WeaponFoley
 volumeMultiplier=0.2
 attenuationOverrides {bAttenuate=false, stereoSpread=5000}
 concurrencyOverrides {maxCount=6}
}
